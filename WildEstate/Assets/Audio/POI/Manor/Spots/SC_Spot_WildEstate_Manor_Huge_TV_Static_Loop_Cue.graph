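# Looping indoor ambient spot cue for the static of a huge TV in the WildEstate Manor POI.
# A mixer layers three looping waves: manor TV static shaped by a slight volume-dip envelope, generic TV static through a pitch-randomizing modulator, and cube-shard idle static.
# Overridden box attenuation adds distance filtering, listener focus and occlusion, with a tight overridden concurrency limit and a reverb submix send.
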SoundCue /WildEstate/Assets/Audio/POI/Manor/Spots/SC_Spot_WildEstate_Manor_Huge_TV_Static_Loop_Cue
root=Mixer_1
nodes [
 {
  id=Enveloper_0
  children [WavePlayer_0]
  properties {pitchCurve {editorCurveData {keys [{time=0}]}}, volumeCurve {editorCurveData {keys [<at 0 1 RCIM_Cubic>,<at 5 0.95 RCIM_Cubic>,<at 10 1>]}}}
 }
 {id=Mixer_1, children [Enveloper_0,Modulator_1,WavePlayer_8], properties {inputVolume [1,1,1]}}
 {id=Modulator_1, children [WavePlayer_4], properties {pitchMax=1.2}}
 {
  id=WavePlayer_0
  properties {bLooping=true, soundWaveAssetPtr=&/Spot_WildEstate_Manor_Huge_TV_Static_Loop}
 }
 {id=WavePlayer_4, properties {bLooping=true, soundWaveAssetPtr=/AmbientAudioContent/Waves/SpottedLoops/Spot_FNM_TV_Static_loop}}
 {id=WavePlayer_8, properties {bLooping=true, soundWaveAssetPtr=/AmbientAudioContent/Waves/SpottedLoops/Spot_MASH_CubeShard_IdleStatic}}
]
properties {
 bOverrideAttenuation=true
 bOverrideConcurrency=true
 maxDistance=1400
 soundClassObject=&Fort_Ambient_Indoors_Athena
 volumeMultiplier=1
 attenuationOverrides {
  attenuationShape=Box
  bEnableListenerFocus=true
  bEnableOcclusion=true
  coneSphereFalloffDistance=750
  coneSphereRadius=1
  falloffDistance=1000
  focusAttackInterpSpeed=2
  focusAzimuth=15
  focusReleaseInterpSpeed=0.5
  hPFFrequencyAtMax=200
  lPFFrequencyAtMax=1000
  lPFRadiusMax=1000
  lPFRadiusMin=500
  nonFocusAzimuth=45
  nonFocusVolumeAttenuation=0.4
  occlusionLowPassFilterFrequency=1000
  occlusionTraceChannel=ECC_GameTraceChannel11
  occlusionVolumeAttenuation=0.5
  attenuationShapeExtents=<200 400 400>
 }
 concurrencyOverrides {maxCount=2, voiceStealReleaseTime=2}
 concurrencySet [/AmbientAudioBase/Concurrency/AmbientConc_3Max_Spot_Default_Concurrency]
 soundSubmixSends [{
   minSendLevel=0.25
   soundSubmix=&Fortnite_Reverb_Default
   customSendLevelCurve {editorCurveData {keys [<at 0 0>,<at 1 1>]}}
  }]
}
